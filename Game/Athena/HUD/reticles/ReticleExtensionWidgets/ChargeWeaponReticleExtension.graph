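# Charged-weapon reticle extension widget that shows a radial charge meter and a text prompt for holding to charge.
# Charge updates drive the Percent parameter on the progress image's dynamic material.
# Construct sets the prompt from the equipped weapon, binds to the controller's view-target change and triggers it once; Destruct unbinds.
# ShowPrompt checks for touch input, reveals the text and restarts a timer, while targeting changes can collapse it.
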WidgetBlueprint /Game/Athena/HUD/reticles/ReticleExtensionWidgets/ChargeWeaponReticleExtension
graphs [
 EventGraph {nodes [
   OnViewTargetChanged!() -> (
    OutputDelegate
    then
    PlayerController:Fort Player Controller Object Reference
    OldViewTarget:Actor Object Reference
    NewViewTarget:Actor Object Reference
   ) {is=title K2Node_CustomEvent}
   > ShowPrompt()
   n1 = Event OnUpdateChargeDisplay!() -> (OutputDelegate, then, ChargePercent:float) {is=title K2Node_Event}
   > n2 = GetDynamicMaterial(self=Get Image_Progress() -> (Image_Progress:Image Object Reference))
   > Is Valid(InputObject=@n2)
   > SetScalarParameterValue(self=@n2, ParameterName=Percent, Value=@n1.ChargePercent)
   n3 = Event Destruct()
   n4 = Create Event()
   n5 = GetOwningPlayer()
   n6 = Cast To FortPlayerController(execute=@n10.then, Object=@n5)
   > n7 = Set OwningFortPlayerController(OwningFortPlayerController:Fort Player Controller Object Reference=@n6) -> (Output_Get:Fort Player Controller Object Reference)
   Unbind Event from On View Target Changed(execute=@n3.then, self:Fort Player Controller Object Reference=@n7, Delegate=@n4) {is=title K2Node_RemoveDelegate}
   n8 = Event OnTargetingChanged!() -> (OutputDelegate, then, bIsADS:bool) {is=title K2Node_Event}
   > Branch(Condition=@n8.bIsADS)
   > HidePrompt()
   Event Construct()
   > n9 = GetHoldPromptText!(
    execute
    self:Fort Weapon Object Reference=GetEquippedWeapon!(self:Fort Weapon Reticle Extension Widget Base Object Reference) -> (ReturnValue:Fort Weapon Object Reference)
   ) -> (then, ReturnValue:Text)
   > n10 = SetPrompt!(execute, self, Text:Text=@n9) -> (then)
   PrintString!(execute, InString=Weapon Fired, bPrintToScreen, bPrintToLog, TextColor, Duration, Key) -> (then)
   OnViewTargetChanged!(
    execute=@n12.then
    self
    PlayerController:Fort Player Controller Object Reference
    OldViewTarget:Actor Object Reference
    NewViewTarget:Actor Object Reference=GetViewTarget!(self:Controller Object Reference=@n5) -> (ReturnValue:Actor Object Reference)
   ) -> (then) {is=title K2Node_CallFunction}
   n11 = Event OnWeaponReloadStarted!() -> (OutputDelegate, then, ReloadTime:float, ReloadType:EFortWeaponReloadType Enum=ReloadWholeClip) {is=title K2Node_Event}
   PrintString(InString=Append(A=Reload started, B="To String (Float)"(InDouble=@n11.ReloadTime)))
   PrintString(InString=Reload Cancelled)
   Event OnWeaponSingleFire() {is=title K2Node_Event}
   Event OnWeaponReloadCancelled() {is=title K2Node_Event}
   n12 = Bind Event to On View Target Changed(execute=@n7.then, self:Fort Player Controller Object Reference=@n7, Delegate=@n4) {is=title K2Node_AddDelegate}
  ]}
 HidePrompt {nodes [
   HidePrompt() {is=title K2Node_FunctionEntry}
   > SetVisibility(self=Get HoldPrompt_CommonText() -> (HoldPrompt_CommonText:Common Text Object Reference), InVisibility=Collapsed)
  ]}
 SetPrompt {nodes [
   n1 = SetPrompt!() -> (then, Text:Text) {is=title K2Node_FunctionEntry}
   > Set Text(Text=@n1, self=Get HoldPrompt_CommonText())
  ]}
 ShowPrompt {nodes [
   ShowPrompt() {is=title K2Node_FunctionEntry}
   > Branch(Condition=IsUsingInputType!(InputType:ECommonInputType Enum=Touch, PlayerController:Player Controller Object Reference=GetOwningPlayer()) -> (ReturnValue:bool))
   > Clear and Invalidate Timer by Handle(Handle=Get PromptTimerHandle() -> (PromptTimerHandle:Timer Handle Structure))
   > SetVisibility(self=Get HoldPrompt_CommonText(), InVisibility=SelfHitTestInvisible)
   > n1 = Set Timer by Event(Delegate=Create Event(), Time=3, bLooping=false)
   > Set PromptTimerHandle(PromptTimerHandle:Timer Handle Structure=@n1) -> (Output_Get:Timer Handle Structure)
  ]}
]
variables [
 OwningFortPlayerController {type=FortPlayerController, default=null}
 PromptTimerHandle {type=TimerHandle}
]
widgets {
 info {
  inheritedWidgetCount=1
  parentClass=/Script/FortniteUI.FortChargedWeaponReticleExtensionWidget
  rootWidgetClass=&Overlay
  widgetCount=6
 }
 widgets [
  Overlay_0 {widgetClassPath=&Overlay, widget=Overlay_0}
  CommonBorder {
   widgetClassPath=&CommonBorder
   bIsVariable=true
   parent=Overlay_0
   widget=CommonBorder
   properties {padding=<ltrb 0 0 0 100>}
   slot {class=OverlaySlot, properties {horizontalAlignment=HAlign_Center, verticalAlignment=VAlign_Center}}
  }
  Image_Progress {
   widgetClassPath=&Image
   bIsVariable=true
   parent=CommonBorder
   widget=Image_Progress
   properties {
    visibility=HitTestInvisible
    brush {
     bIsDynamicallyLoaded=false
     drawAs=Image
     imageType=NoImage
     mirroring=NoMirror
     resourceName=None
     resourceObject=/Game/Athena/HUD/Materials/M_UI_RadialProgress_Ammo
     tiling=NoTile
     imageSize=<64 64>
     margin=<ltrb 0 0 0 0>
     outlineSettings {
      bUseBrushTransparency=false
      roundingType=HalfHeightRadius
      width=0
      color {colorUseRule=UseColor_Specified, specifiedColor=<rgba 0 0 0 0>}
      cornerRadii=<0 0 0 0>
     }
     tintColor {colorUseRule=UseColor_Specified, specifiedColor=<rgba 1 1 1 1>}
     uVRegion {bIsValid=false, max=<0 0>, min=<0 0>}
    }
    colorAndOpacity=<rgba 1 1 1 0.8>
    renderTransform {angle=-45, scale=<-1 1>, shear=<0 0>, translation=<0 0>}
   }
   slot {class=BorderSlot, properties {padding=<ltrb 0 0 0 100>}}
  }
  HoldPrompt {
   widgetClassPath=&HorizontalBox
   parent=Overlay_0
   widget=HoldPrompt
   properties {renderTransform {angle=0, scale=<1 1>, shear=<0 0>, translation=<200 59>}}
   slot {class=OverlaySlot, properties {horizontalAlignment=HAlign_Center, verticalAlignment=VAlign_Center}}
  }
  HoldPrompt_CommonText {
   widgetClassPath=&CommonTextBlock
   bIsVariable=true
   parent=HoldPrompt
   widget=HoldPrompt_CommonText
   properties {
    style=/FortUILibrary/Styles/Typography/_TextStyle_TagNormal._TextStyle_TagNormal_C
    visibility=Collapsed
    font {
     bForceMonospaced=false
     bMaterialIsStencil=false
     fontMaterial=null
     fontObject=&BurbankSmall
     letterSpacing=0
     monospacedWidth=1
     size=15
     skewAmount=0
     typefaceFontName=Black
     outlineSettings {
      bApplyOutlineToDropShadows=false
      bMiteredCorners=false
      bSeparateFillAlpha=false
      outlineMaterial=null
      outlineSize=0
      outlineColor=<rgba 0 0 0 1>
     }
    }
    shadowOffset=<0 0>
    strikeBrush {
     bIsDynamicallyLoaded=false
     drawAs=NoDrawType
     imageType=NoImage
     mirroring=NoMirror
     resourceName=None
     resourceObject=null
     tiling=NoTile
     imageSize=<32 32>
     margin=<ltrb 0 0 0 0>
     outlineSettings {
      bUseBrushTransparency=false
      roundingType=HalfHeightRadius
      width=0
      color {colorUseRule=UseColor_Specified, specifiedColor=<rgba 0 0 0 0>}
      cornerRadii=<0 0 0 0>
     }
     tintColor {colorUseRule=UseColor_Specified, specifiedColor=<rgba 1 1 1 1>}
     uVRegion {bIsValid=false, max=<0 0>, min=<0 0>}
    }
   }
   slot {class=HorizontalBoxSlot, properties {verticalAlignment=VAlign_Center, padding=<ltrb 4 0 0 0>}}
  }
  TutorialHighlight {widgetClassPath=&TutorialHighlight}
 ]
}
