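# Reticle extension widget that shows the minigun's overheat level as a radial meter, using the ammo radial-progress material.
# On Construct it zeroes the material's Percent, casts the equipped weapon to B_Minigun_Athena and binds to its overheat-value-changed event; Destruct unbinds.
# Set Progress, shaped as that event's handler, writes CurrentValue / MaxValue to Percent, ignoring overheat state and cooldown.
# The HeatMeterMaterialRef variable is unused in the graphs shown.
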WidgetBlueprint /Game/Athena/HUD/reticles/ReticleExtensionWidgets/WBP_Overheet_Reticle
graphs [
 EventGraph {nodes [
   Event Construct()
   > n1 = GetDynamicMaterial(self=Get Image_Magazine() -> (Image_Magazine:Image Object Reference))
   > SetScalarParameterValue(self=@n1, ParameterName=Percent, Value=0)
   > n2 = Cast To B_Minigun_Athena!(
    execute
    Object=GetEquippedWeapon!(self:Fort Weapon Reticle Extension Widget Base Object Reference) -> (ReturnValue:Fort Weapon Object Reference)
   ) -> (then, CastFailed, AsB Minigun Athena:B Minigun Athena Object Reference) {is=title K2Node_DynamicCast}
   > n3 = Set WeaponRef(WeaponRef:B Minigun Athena Object Reference=@n2) -> (Output_Get:B Minigun Athena Object Reference)
   > Bind Event to On Overheat Value Changed(self:Fort Weapon Ranged Object Reference=@n3, Delegate=Create Event()) {is=title K2Node_AddDelegate}
   Event Destruct()
   > Unbind Event from On Overheat Value Changed(self:Fort Weapon Ranged Object Reference=Get WeaponRef() -> (WeaponRef:B Minigun Athena Object Reference), Delegate=Create Event()) {is=title K2Node_RemoveDelegate}
  ]}
 Set Progress {nodes [
   n1 = Set Progress!() -> (then, CurrentValue:float=~, MaxValue:float=~, State:EFortWeaponOverheatState Enum=None, bIsInCooldown:bool=~) {is=title K2Node_FunctionEntry}
   > n2 = GetDynamicMaterial(self=Get Image_Magazine())
   > SetScalarParameterValue(self=@n2, ParameterName=Percent, Value=float / float(A=@n1, B=@n1.MaxValue))
  ]}
]
variables [
 HeatMeterMaterialRef {type=MaterialInstanceDynamic}
 WeaponRef {type=B_Minigun_Athena_C, default=null}
]
widgets {
 info {
  inheritedWidgetCount=1
  parentClass=/Script/FortniteUI.FortWeaponReticleExtensionWidgetBase
  rootWidgetClass=&Overlay
  widgetCount=3
 }
 widgets [
  Root {widgetClassPath=&Overlay, widget=Root}
  Image_Magazine {
   widgetClassPath=&Image
   bIsVariable=true
   parent=Root
   widget=Image_Magazine
   properties {
    brush {
     bIsDynamicallyLoaded=false
     drawAs=Image
     imageType=NoImage
     mirroring=NoMirror
     resourceName=None
     resourceObject=/Game/Athena/HUD/Materials/M_UI_RadialProgress_Ammo
     tiling=NoTile
     imageSize=<80 80>
     margin=<ltrb 0 0 0 0>
     outlineSettings {
      bUseBrushTransparency=false
      roundingType=HalfHeightRadius
      width=0
      color {colorUseRule=UseColor_Specified, specifiedColor=<rgba 0 0 0 0>}
      cornerRadii=<0 0 0 0>
     }
     tintColor {colorUseRule=UseColor_Specified, specifiedColor=<rgba 1 1 1 1>}
     uVRegion {bIsValid=false, max=<0 0>, min=<0 0>}
    }
    renderTransform {angle=180, scale=<0.9 0.9>, shear=<0 0>, translation=<100 100>}
   }
   slot {class=OverlaySlot, properties {horizontalAlignment=HAlign_Center, verticalAlignment=VAlign_Center}}
  }
  TutorialHighlight {widgetClassPath=&TutorialHighlight}
 ]
}
