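# Reticle extension widget that shows the equipped weapon's magazine as a segmented radial ammo bar beside the crosshair.
# A dynamic material receives fill from local ammo changes, segment count from clip size, and an ADS flag.
# Ammo changes and aiming show the bar, then timers play a delayed fade-out.
# Reloads hold it visible, deferring the fade until the reload timer finishes or is cancelled.
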WidgetBlueprint /Game/Athena/HUD/reticles/ReticleExtensionWidgets/WBP_Reticle_MagazineOverride
animations [FadeOut {
  binding_count=1
  length_seconds=0.4969
  object=&self:FadeOut
  sequence {bindings [Image_Magazine {
     id=#1
     possessed_class=Image
     tracks [{
       type=MovieSceneFloatTrack
       display_name=Render Opacity
       sections [{type=MovieSceneFloatSection, channels [None {type=MovieSceneScriptingFloatChannel, keys [[0,1],[9.937,0]]}]}]
      }]
    }]}
 }]
graphs [
 EventGraph {nodes [
   n1 = OnReloadTimerFinished() {is=title K2Node_CustomEvent}
   Clear and Invalidate Timer by Handle(execute=[@n1.then, @n10.then], Handle=Get ReloadTimer() -> (ReloadTimer:Timer Handle Structure))
   > QueueFadeOut()
   QueueFadeOut() {is=title K2Node_CustomEvent}
   > n2 = Branch(Condition=Is Valid Timer Handle!(Handle:Timer Handle Structure=Get ReloadTimer()) -> (ReturnValue:bool))
   n3 = FadeIn(execute=@n2.else)
   n4 = PlayFadeOut() {is=title K2Node_CustomEvent}
   n5 = Set Timer by Event(execute=@n3.then, Delegate=@n4, Time=Get TimeToOutro() -> (TimeToOutro:double), bLooping=false)
   > Set WaitToFade(WaitToFade:Timer Handle Structure=@n5) -> (Output_Get:Timer Handle Structure)
   n6 = Event OnWeaponReloadStarted!() -> (OutputDelegate, then, ReloadTime:float, ReloadType:EFortWeaponReloadType Enum=ReloadWholeClip) {is=title K2Node_Event}
   > n7 = Set Timer by Event(Delegate=@n1, Time=@n6.ReloadTime, bLooping=false)
   > Set ReloadTimer(ReloadTimer:Timer Handle Structure=@n7) -> (Output_Get:Timer Handle Structure)
   > FadeIn()
   Event Construct()
   > Parent: Construct()
   > n8 = Set EquippedWeapon(EquippedWeapon:Fort Weapon Object Reference=GetEquippedWeapon!(self:Fort Weapon Reticle Extension Widget Base Object Reference) -> (ReturnValue:Fort Weapon Object Reference)) -> (Output_Get:Fort Weapon Object Reference)
   > Is Valid(InputObject=@n8)
   > Bind Event to On Local Ammo Changed(self:Fort Weapon Object Reference=Get EquippedWeapon() -> (EquippedWeapon:Fort Weapon Object Reference), Delegate=Create Event()) {is=title K2Node_AddDelegate}
   n9 = Get EquippedWeapon()
   > OnLocalAmmoChanged!(
    execute
    self
    LocalCount:int=GetMagazineAmmoCount!(self:Fort Weapon Object Reference=@n9) -> (ReturnValue:int=0)
    LocalRemaining:int=GetRemainingAmmo!(self:Fort Weapon Object Reference=@n9) -> (ReturnValue:int=0)
   ) -> (then)
   n10 = Event OnWeaponReloadCancelled() {is=title K2Node_Event}
   Event Destruct()
   > Parent: Destruct()
   > Clear and Invalidate Timer by Handle(Handle=Get ReloadTimer())
   > Clear and Invalidate Timer by Handle(Handle=Get WaitToFade() -> (WaitToFade:Timer Handle Structure))
   n11 = Get EquippedWeapon()
   > Is Valid(InputObject=@n11)
   > Unbind all Events from On Local Ammo Changed!(execute:Exec, self:Fort Weapon Object Reference=@n11) -> (then:Exec) {is=title K2Node_ClearDelegate}
   n12 = Event PreConstruct()
   > n13 = Branch(Condition=@n12.IsDesignTime)
   SetBrushFromMaterial(
    execute=[@n14.then_0, @n13.then]
    self=Get Image_Magazine() -> (Image_Magazine:Image Object Reference)
    Material=Get AmmoBarMaterial() -> (AmmoBarMaterial:Material Instance Object Reference)
   )
   Event OnInitialized()
   > n14 = Sequence()
   n15 = GetDynamicMaterial(execute=@n14.then_1, self=Get Image_Magazine())
   > n16 = Set MagazineDynamicMaterial(MagazineDynamicMaterial:Material Instance Dynamic Object Reference=@n15) -> (Output_Get:Material Instance Dynamic Object Reference)
   > UpdateAmmoSegments()
   > SetScalarParameterValue(self=@n16, ParameterName=Get IsInADSParam() -> (IsInADSParam:Name=None), Value=0)
   n17 = Event OnTargetingChanged!() -> (OutputDelegate, then, bIsADS:bool) {is=title K2Node_Event}
   > n18 = Sequence()
   > Branch(Condition=@n17.bIsADS)
   > QueueFadeOut() {is=title K2Node_CallFunction}
   n19 = Get MagazineDynamicMaterial() -> (MagazineDynamicMaterial:Material Instance Dynamic Object Reference)
   Is Valid(exec=@n18.then_1, InputObject=@n19)
   > SetScalarParameterValue(self=@n19, ParameterName=Get IsInADSParam() -> (IsInADSParam=None), Value=SelectFloat(A=1, B=0, bPickA=@n17.bIsADS))
   Event OnAmmoClipAttributeChanged!() -> (OutputDelegate, then, AmmoClipValue:float) {is=title K2Node_Event}
   > UpdateAmmoSegments()
   PlayAnimation(execute=@n4.then, InAnimation=Get FadeOut() -> (FadeOut:Widget Animation Object Reference))
  ]}
 FadeIn {nodes [
   FadeIn() {is=title K2Node_FunctionEntry}
   > SetRenderOpacity(self=Get Image_Magazine(), InOpacity=1)
   > Clear and Invalidate Timer by Handle(Handle=Get WaitToFade())
   > StopAnimation(InAnimation=Get FadeOut())
  ]}
 OnLocalAmmoChanged {nodes [
   n1 = OnLocalAmmoChanged!() -> (then, LocalCount:int, LocalRemaining:int) {is=title K2Node_FunctionEntry}
   > QueueFadeOut()
   > ResetAmmo!(execute, self, LocalCount:int=@n1, LocalRemaining:int=@n1.LocalRemaining) -> (then)
  ]}
 ResetAmmo {nodes [
   ResetAmmo!() -> (then, LocalCount:int, LocalRemaining:int) {is=title K2Node_FunctionEntry}
   n1 = GetBulletsPerClip!(self:Fort Weapon Object Reference=Get EquippedWeapon()) -> (ReturnValue:int=0)
   n2 = Get LocalCount() -> (LocalCount:int=0)
   > SetScalarParameterValue(
    self=Get MagazineDynamicMaterial()
    ParameterName=Progress
    Value=float / float(A="To Float (Integer)"(InInt=@n2), B="To Float (Integer)"(InInt=@n1))
   )
  ]}
 UpdateAmmoSegments {nodes [
   UpdateAmmoSegments() {is=title K2Node_FunctionEntry}
   n1 = Get MagazineDynamicMaterial()
   > Is Valid(InputObject=@n1)
   n2 = GetBulletsPerClip(self=GetEquippedWeapon()) -> (ReturnValue=0)
   n3 = int * int(A:int=@n2, B:int=4) -> (ReturnValue:int=0) {is=title K2Node_CommutativeAssociativeBinaryOperator}
   > SetScalarParameterValue(
    self=@n1
    ParameterName=Segments
    Value=Select(Option 0:double="To Float (Integer)"(InInt=@n3), Option 1:double=1, Index="Equal (Integer)"(A=@n2, B=1) {is=title K2Node_CallFunction}) -> (ReturnValue:double)
   )
  ]}
]
variables [
 AmmoBarMaterial {type=MaterialInstance, default=/Game/Athena/HUD/Materials/M_UI_RadialProgress_Ammo}
 EquippedWeapon {type=FortWeapon, default=null}
 IsInADSParam {type=string, default=IsInADS}
 MagazineDynamicMaterial {type=MaterialInstanceDynamic}
 ReloadTimer {type=TimerHandle}
 TimeToOutro {type=number, default=1.5}
 WaitToFade {type=TimerHandle}
]
widgets {
 info {
  inheritedWidgetCount=1
  parentClass=/Script/FortniteUI.FortWeaponReticleExtensionWidgetBase
  rootWidgetClass=&Overlay
  widgetCount=4
 }
 widgets [
  Overlay_0 {widgetClassPath=&Overlay, widget=Overlay_0}
  NamedSlot_ReticleOverlay {
   widgetClassPath=&NamedSlot
   bIsVariable=true
   parent=Overlay_0
   widget=NamedSlot_ReticleOverlay
   slot {class=OverlaySlot, properties {horizontalAlignment=HAlign_Fill, verticalAlignment=VAlign_Fill}}
  }
  Image_Magazine {
   widgetClassPath=&Image
   bIsVariable=true
   parent=Overlay_0
   widget=Image_Magazine
   properties {
    brush {
     bIsDynamicallyLoaded=false
     drawAs=Image
     imageType=NoImage
     mirroring=NoMirror
     resourceName=None
     resourceObject=&M_UI_RadialProgress_Ammo
     tiling=NoTile
     imageSize=<80 80>
     margin=<ltrb 0 0 0 0>
     outlineSettings {
      bUseBrushTransparency=false
      roundingType=HalfHeightRadius
      width=0
      color {colorUseRule=UseColor_Specified, specifiedColor=<rgba 0 0 0 0>}
      cornerRadii=<0 0 0 0>
     }
     tintColor {colorUseRule=UseColor_Specified, specifiedColor=<rgba 1 1 1 1>}
     uVRegion {bIsValid=false, max=<0 0>, min=<0 0>}
    }
    renderTransform {angle=180, scale=<0.9 0.9>, shear=<0 0>, translation=<100 100>}
   }
   slot {class=OverlaySlot, properties {horizontalAlignment=HAlign_Center, verticalAlignment=VAlign_Center}}
  }
  TutorialHighlight {widgetClassPath=&TutorialHighlight}
 ]
}
